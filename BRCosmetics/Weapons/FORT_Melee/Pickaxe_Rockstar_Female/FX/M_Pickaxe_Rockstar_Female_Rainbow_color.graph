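# Translucent rainbow effect material for the female Rockstar pickaxe, usable on Niagara mesh particles and skeletal meshes.
# Panning noise modulates a desaturatable rainbow texture, while UV-based sine masks shape opacity and edge glow, softened by depth fade.
# Scalar lerps select UV set, edge brightness and skeletal-mesh mode, which bypasses particle color.
# Optional WPO adds a time-based wave and a V-weighted pull from Niagara dynamic parameters, whose target repeats X across all axes.
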Material /BRCosmetics/Weapons/FORT_Melee/Pickaxe_Rockstar_Female/FX/M_Pickaxe_Rockstar_Female_Rainbow_color
graph {
 properties {
  bUseMaterialAttributes=true
  bUsedWithMeshDeformer=true
  bUsedWithNiagaraMeshParticles=true
  bUsedWithSkeletalMesh=true
  blendMode=BLEND_Translucent
  pixelDepthOffsetMode=PDOM_Legacy
  usedShadingModels=Default Lit
 }
 nodes [
  n1 = ParticleColor()
  n2 = TextureCoordinate() {vTiling=0.3}
  n3 = TextureCoordinate() {coordinateIndex=1, vTiling=0.3}
  n4 = ScalarParameter() {parameterName=UseUVID0-1}
  n5 = TextureSampleParameter2D(UVs=Panner(Coordinate=LinearInterpolate(A=@n2, B=@n3, Alpha=@n4)) {speedY=0.3}) {parameterName=Noise, samplerType=SAMPLERTYPE_Grayscale, texture=&LowResBlurredNoise}
  n6 = MaterialFunctionCall(XY=LinearInterpolate(A=TextureCoordinate(), B=TextureCoordinate() {coordinateIndex=1}, Alpha=ScalarParameter() {parameterName=UseUVID0-1})) {materialFunction=&Swizzle}
  n7 = Multiply(
   A=Multiply(A=LinearInterpolate(Alpha=@n5.R) {constA=0.08}) {constB=3}
   B=Desaturation(
    None=TextureSampleParameter2D(UVs=@n6.YX) {parameterName=Basic_texture, texture=&/T_RockStar_Rainbow_color}.RGB
    Fraction=ScalarParameter() {defaultValue=-1, parameterName=Desaturation}
   )
  )
  n8 = TextureCoordinate()
  n9 = LinearInterpolate(
   A=VectorParameter() {defaultValue=<rgba 30 0 0.5405 1>}.RGB
   B=VectorParameter() {parameterName=Param2, defaultValue=<rgba 20 0 6.5428 1>}.RGB
   Alpha=Round(None=ComponentMask(None=@n8) {r=true})
  )
  n10 = ScalarParameter() {parameterName=Use_edge_brightness}
  n11 = Sine(None=ComponentMask(None=TextureCoordinate()) {r=true}) {period=2}
  n12 = Saturate(None=Power(Base=OneMinus(None=@n11)) {constExponent=5})
  n13 = LinearInterpolate(
   A=Multiply(A=@n1.RGB, B=LinearInterpolate(A=@n7, B=@n9, Alpha=@n10))
   B=Add(A=Multiply(A=@n9, B=@n12), B=@n7)
   Alpha=ScalarParameter() {parameterName=UsetoSkeletalMesh}
  )
  n14 = TextureCoordinate()
  n15 = Power(Base=Sine(None=ComponentMask(None=@n14) {r=true}) {period=2}) {constExponent=1}
  n16 = Sine(None=ComponentMask(None=TextureCoordinate()) {r=true}) {period=2}
  n17 = LinearInterpolate(Alpha=OneMinus(None=Power(Base=@n16) {constExponent=1})) {constA=0.2}
  n18 = TextureCoordinate()
  n19 = Multiply(
   A=Multiply(A=Multiply(A=@n15, B=@n17) {constB=0.7}) {constB=2.5}
   B=LinearInterpolate(A=ScalarParameter() {parameterName=Mask}, Alpha=ComponentMask(None=@n18) {g=true}) {constA=0.1}
  )
  n20 = Sine(None=ComponentMask(None=TextureCoordinate()) {r=true}) {period=2}
  n21 = Saturate(None=Power(Base=OneMinus(None=@n20)) {constExponent=5})
  n22 = LinearInterpolate(A=@n19, B=@n21, Alpha=ScalarParameter() {parameterName=Use_edge_brightness})
  n23 = DepthFade(
   Opacity=LinearInterpolate(A=Multiply(A=@n1.A, B=@n22), B=Add(A=@n19, B=@n21), Alpha=ScalarParameter() {parameterName=UsetoSkeletalMesh})
   FadeDistance=ScalarParameter() {parameterName=DepthFade}
  )
  n24 = TextureCoordinate()
  n25 = Add(A=Multiply(A=ComponentMask(None=@n24) {g=true}) {constB=0.7}, B=Time())
  n26 = ComponentMask(None=TextureCoordinate()) {g=true}
  n27 = Multiply(A=Multiply(A=Sine(None=@n25), B=OneMinus(None=@n26)), B=VectorParameter() {defaultValue=<rgba 30 0 0.5405 1>}.RGB)
  n28 = ScalarParameter() {defaultValue=0.02, parameterName=Param_4}
  n29 = DynamicParameter() {desc=Parameters set by Niagara, parameterIndex=1, paramNames [X,Y,Z,WPOIntensity]}
  n30 = Subtract(
   A=MaterialFunctionCall(X=@n29.X, Y=@n29.X, Z=@n29.X) {materialFunction=&MakeFloat3}.Result
   B=ParticlePositionWS()
  )
  n31 = ScalarParameter() {defaultValue=4, group=Interpolate, parameterName=Interpolate_Clamp_Max}
  n32 = StaticSwitchParameter(
   True=Clamp(None=@n30, Min=ScalarParameter() {group=Interpolate, parameterName=Interpolate_Clamp_Min}, Max=@n31) {maxDefault=1000, minDefault=-1000}
   False=@n30
  ) {group=Switch, parameterName=Switch_InterpolateClamp}
  n33 = OneMinus(None=TextureCoordinate())
  n34 = Multiply(
   A=@n29.WPOIntensity
   B=Power(Base=ComponentMask(None=@n33) {g=true}, Exp=ScalarParameter() {defaultValue=3, group=Interpolate, parameterName=Interpolate_Power})
  )
  n35 = LinearInterpolate(A=Constant(), B=Add(A=Multiply(A=@n27, B=@n28), B=Multiply(A=@n32, B=@n34)), Alpha=ScalarParameter() {parameterName=WPO})
  MaterialFunctionCall(In=SetMaterialAttributes!(
    MaterialAttributes
    Emissive Color=NamedRerouteDeclaration(None=@n13) {name=emissiveColor, nodeColor=<rgba 1 0.9609 0 1>}
    Opacity=NamedRerouteDeclaration(None=@n23) {name=opacity, nodeColor=<rgba 0 1 0.4609 1>}
    World Position Offset=NamedRerouteDeclaration(None=@n35) {name=wpo, nodeColor=<rgba 1 0.5156 0 1>}
   ) {attributeSetTypes [#1, #2, #3]}) {materialFunction=&MF_TechArtOverrides_accessories}
  Multiply(A=@n30, B=@n31) {constA=-1}
 ]
}
